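# Sound cue for the LastVoiceDive cosmetic pickaxe's enemy impact, playing drum-style kick and snare hits.
# A switch on the integer parameter LastVoiceDrive steps through alternating kick and snare samples, four of each, with the first kick also filling the first slot.
# A short fixed delay precedes playback.
# It uses the shared pickaxe impact attenuation, property sheet and impacts sound class.
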SoundCue /BRCosmetics/Sounds/PickAxes/LastVoiceDive/PickaxeImpactEnemy_Last_VoiceDive_Cue
root=Delay_0
nodes [
 {id=Delay_0, children [Switch_0], properties {delayMax=0.1, delayMin=0.1}}
 {
  id=Switch_0
  children [WavePlayer_1,WavePlayer_1,WavePlayer_5,WavePlayer_2,WavePlayer_6,WavePlayer_3,WavePlayer_7,WavePlayer_4,WavePlayer_0]
  properties {intParameterName=LastVoiceDrive}
 }
 {id=WavePlayer_0, properties {soundWaveAssetPtr=&/Pickaxe_LastVoiceDive_Snare_Impact_04}}
 {id=WavePlayer_1, properties {soundWaveAssetPtr=&/Pickaxe_LastVoiceDive_Kick_Impact_01}}
 {id=WavePlayer_2, properties {soundWaveAssetPtr=&/Pickaxe_LastVoiceDive_Kick_Impact_02}}
 {id=WavePlayer_3, properties {soundWaveAssetPtr=&/Pickaxe_LastVoiceDive_Kick_Impact_03}}
 {id=WavePlayer_4, properties {soundWaveAssetPtr=&/Pickaxe_LastVoiceDive_Kick_Impact_04}}
 {id=WavePlayer_5, properties {soundWaveAssetPtr=&/Pickaxe_LastVoiceDive_Snare_Impact_01}}
 {id=WavePlayer_6, properties {soundWaveAssetPtr=&/Pickaxe_LastVoiceDive_Snare_Impact_02}}
 {id=WavePlayer_7, properties {soundWaveAssetPtr=&/Pickaxe_LastVoiceDive_Snare_Impact_03}}
]
properties {
 attenuationSettings=/Game/Sounds/Attenuation/Fort_Pickaxe_Impact_Att
 audioPropertiesSheet=/BRCosmetics/Sounds/Settings/PropertySheets/Pickaxe/APS_Cosmetic_Pickaxe_Impact
 maxDistance=4300
 soundClassObject=&Fort_Impacts
 concurrencySet [null]
}
